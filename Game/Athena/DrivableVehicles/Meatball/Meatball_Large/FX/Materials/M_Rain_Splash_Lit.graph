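# Translucent Default Lit material for rain splash sprites in the Meatball_Large vehicle FX, usable with Niagara sprites.
# A panning noise map distorts splash texture UVs outside mobile, and the splash normal drives a wobbling sky cubemap reflection plus a sun glint.
# Reflection and particle color are darkened by the splash mask.
# Opacity combines mask, scrolling noise, particle alpha and an Erode dynamic parameter, with depth fade skipped on mobile and forward shading.
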Material /Game/Athena/DrivableVehicles/Meatball/Meatball_Large/FX/Materials/M_Rain_Splash_Lit
graph {
 properties {
  bTangentSpaceNormal=false
  bUsedWithNanite=true
  bUsedWithNiagaraSprites=true
  blendMode=BLEND_Translucent
  pixelDepthOffsetMode=PDOM_Legacy
  translucencyLightingMode=TLM_VolumetricPerVertexNonDirectional
  translucencyPass=MTP_BeforeDOF
  translucentSelfShadowDensityScale=0
  translucentSelfShadowSecondDensityScale=0
  translucentShadowDensityScale=0
  usedShadingModels=Default Lit
 }
 nodes [
  n1 = Panner(Coordinate=TextureCoordinate() {uTiling=2, vTiling=2}) {speedY=-0.06}
  n2 = TextureCoordinate()
  n3 = TextureSample(UVs=Panner(Coordinate=TextureCoordinate() {uTiling=2, vTiling=2}) {speedY=-0.03}) {texture=/Game/Global/Textures/Default/Tileable_Patterns/T_Tilling_Distortion_Map_01}
  n4 = ComponentMask(None=@n3.G) {g=true, r=true}
  n5 = ShadingPathSwitch(Default=Add(A=@n2, B=Multiply(A=@n4) {constB=0.1}), Mobile=@n2) {desc=Migrated from FeatureLevelSwitch}
  n6 = ReflectionVectorWS(CustomWorldNormal=Transform(None=TextureSample(UVs=@n5) {samplerType=SAMPLERTYPE_Normal, texture=/Game/Effects/Fort_Effects/Textures/Water/T_waterSplash_N}.RGB))
  n7 = TextureSampleParameterCube!(
   UVs=Add(
    A=TextureSample(UVs=@n1) {samplerType=SAMPLERTYPE_Normal, texture=/Game/Global/Textures/Default/Tileable_Patterns/T_Tilling_Distortion_Map_01_n}.RGB
    B=@n6
   )
   Apply View MipBias
  ) {parameterName=Cubemap, samplerType=SAMPLERTYPE_LinearColor, texture=/Game/Packages/Fortress_Sky/TexturesHDR/T_AthenaSkylight}
  n8 = ComponentMask(None=MaterialFunctionCall() {materialFunction=/Engine/Functions/GetPostProcessFunctions/GetAmbientCubemapTint}.AmbientCubemapTint) {b=true, g=true, r=true}
  n9 = CollectionParameter() {
   collection=&FortniteMaterialParameters
   parameterName=SunAndMoonModelDirectionalVector
  }
  n10 = ComponentMask(None=@n9) {b=true, g=true, r=true}
  n11 = Power(Base=Clamp(None=DotProduct(A=@n10, B=@n6))) {constExponent=20}
  n12 = Power(Base=ComponentMask(None=CollectionParameter() {collection=&FortniteMaterialParameters, parameterName=SunLightColor}) {b=true, g=true, r=true}) {constExponent=0.1}
  n13 = Add(
   A=Multiply(A=Multiply(A=@n7.RGB, B=@n8), B=VectorParameter() {parameterName=Cube Map Multiplier, defaultValue=<rgba 0.75 0.75 0.75 1>}.RGB)
   B=Multiply(A=Clamp(None=@n11) {maxDefault=5}, B=Clamp(None=@n12))
  )
  n14 = ParticleColor()
  n15 = TextureSample(UVs=@n5) {samplerType=SAMPLERTYPE_LinearColor, texture=/Game/Effects/Fort_Effects/Textures/Water/T_waterSplash}
  n16 = OneMinus(None=@n15.G)
  n17 = Add(A=Multiply(A=Clamp(None=@n16)) {constB=0.125}) {constB=0.6}
  Clamp(None=LinearInterpolate(A=Add(A=@n13, B=@n14.RGB), Alpha=Clamp(None=@n17)) {constB=0}) {maxDefault=100}
  n18 = DynamicParameter() {paramNames [Erode,Param2,Param3,Param4]}
  n19 = Multiply(
   A=Power(Base=Multiply(A=@n15.G, B=@n3.G), Exp=LinearInterpolate(Alpha=@n18.Erode) {constA=5, constAlpha=0}) {constExponent=50}
   B=LinearInterpolate(Alpha=@n18.Erode) {constAlpha=0, constB=200}
  )
  n20 = Multiply(A=ParticleColor().A, B=Multiply(A=@n15.G, B=Clamp(None=@n19)))
  n21 = DepthFade(Opacity=@n20, FadeDistance=ScalarParameter() {defaultValue=15, parameterName=DepthFade})
  Clamp(None=ShadingPathSwitch(Default=ShadingPathSwitch(Default=@n21, Mobile=@n20) {desc=Migrated from FeatureLevelSwitch}, Forward=@n20, Mobile=@n20)) {maxDefault=100}
 ]
}
