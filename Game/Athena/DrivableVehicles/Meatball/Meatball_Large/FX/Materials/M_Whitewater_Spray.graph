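# Masked subsurface material for Niagara sprite whitewater spray on the large Meatball drivable vehicle.
# A desaturated sky cubemap, sampled by a tangent-space reflection vector, is tinted by particle color.
# Reused dirt-clod textures supply the normal and bump-offset alpha shape, which particle alpha erodes by subtracting its inverse.
# A "Use Snow Color in SSS?" static switch selects snow tint or white, and clod alpha or a scalar.
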Material /Game/Athena/DrivableVehicles/Meatball/Meatball_Large/FX/Materials/M_Whitewater_Spray
graph {
 properties {
  bUsedWithNiagaraSprites=true
  blendMode=BLEND_Masked
  pixelDepthOffsetMode=PDOM_Legacy
  shadingModel=MSM_Subsurface
  usedShadingModels=Subsurface
 }
 nodes [
  Constant() {r=1}
  n1 = Transform(None=ReflectionVectorWS()) {transformSourceType=TRANSFORMSOURCE_World, transformType=TRANSFORM_Tangent}
  n2 = ParticleColor()
  Multiply(
   A=Desaturation(None=TextureSampleParameterCube!(UVs=@n1, Apply View MipBias) {parameterName=Param, texture=/Game/Packages/Fortress_Sky/SkyDome/MaterialInstances/SkyDome_Apollo/Cubemap_Apollo_Day}.RGB)
   B=@n2.RGB
  )
  ScalarParameter() {defaultValue=0.1, parameterName=Roughness}
  StaticSwitchParameter(
   True=VectorParameter() {parameterName=Snow Color, defaultValue=<rgba 0.1347 0.3106 0.7604 0>}.RGB
   False=Constant3Vector() {constant=<rgba 1 1 1 0>}
  ) {parameterName=Use Snow Color in SSS?}
  n3 = TextureSample(UVs=BumpOffset(Height=TextureSample() {texture=/Game/Effects/Fort_Effects/Textures/Dirt/T-DirtClodCluster_D}.A)) {texture=&T-DirtClodCluster_D}
  StaticSwitchParameter(True=@n3.A, False=ScalarParameter() {defaultValue=1}) {parameterName=Use Snow Color in SSS?}
  Subtract(A=@n2.A, B=OneMinus(None=@n3.A))
  TextureSample() {samplerType=SAMPLERTYPE_Normal, texture=/Game/Effects/Fort_Effects/Textures/Dirt/T-DirtClodCluster_N}
 ]
}
